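# Translucent unlit Niagara mesh-particle material for animated letters in the DriedSilk pickaxe FX.
# A GrowShrink dynamic parameter sweeps against vertex-color green to form a per-vertex mask that scales center-relative offsets and blends a camera offset.
# A sine wiggle, phased by vertex-color red and modulated by particle age and WigglePower, joins those offsets.
# The mask also drives a vertex-interpolated glow over the texture, tinted by particle color.
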Material /BRCosmetics/Weapons/FORT_Melee/Pickaxe_DriedSilk/FX/M_DriedSlik_Pickaxe_Letters_Animated
graph {
 properties {
  bCastRayTracedShadows=false
  bUsedWithNiagaraMeshParticles=true
  blendMode=BLEND_Translucent
  pixelDepthOffsetMode=PDOM_Legacy
  shadingModel=MSM_Unlit
  twoSided=true
  usedShadingModels=Unlit
 }
 nodes [
  n1 = VertexColor()
  n2 = DynamicParameter() {defaultValue=<rgba 1 1 1 0>, paramNames [GrowShrink,CameraOffset,WigglePower,Param4]}
  n3 = OneMinus(None=Multiply(A=@n2.GrowShrink) {constB=2})
  n4 = Clamp(None=LinearInterpolate(Alpha=Subtract(A=@n1.G, B=@n3)) {constA=-1}) {minDefault=-1}
  LinearInterpolate(
   B=MaterialFunctionCall(Offset Amount=DynamicParameter() {defaultValue=<rgba 1 1 1 0>, paramNames [GrowShrink,CameraOffset,WigglePower,Param4]}.CameraOffset) {materialFunction=/Engine/Functions/Engine_MaterialFunctions02/WorldPositionOffset/CameraOffset}.Camera Offset
   Alpha=Clamp(None=@n4)
  )
  n5 = Sine(None=Saturate(None=Multiply(A=@n4) {constB=0.5}))
  n6 = TextureSampleParameter2D() {parameterName=Texture, texture=&/T_DriedSilk_Impact_Resized}
  n7 = Add(A=Multiply(A=VectorParameter() {parameterName=GlowColor, defaultValue=<rgba 40 2 2.6333 0>}.RGB, B=VertexInterpolator(VS=@n5).PS), B=@n6.RGB)
  n8 = ScalarParameter() {defaultValue=1, parameterName=Use_Glow_FadeIn}
  MaterialFunctionCall(Emissive=Multiply(A=LinearInterpolate(A=@n6.RGB, B=@n7, Alpha=@n8), B=ParticleColor().RGB)) {materialFunction=&MF_FX_EmissiveAdjustments}
  n9 = Add(A=Subtract(A=WorldPosition().XYZ, B=ObjectPositionWS()), B=VectorParameter() {parameterName=Center Offset}.RGB)
  n10 = Add(A=Time(), B=VertexColor().R)
  n11 = ScalarParameter() {defaultValue=2.2, parameterName=WiggleSpeed}
  n12 = ConstantBiasScale(None=Sine(None=Multiply(A=@n10, B=@n11) {constB=1.5})) {bias=1.5}
  n13 = Multiply(A=ScalarParameter() {defaultValue=12, parameterName=Wiggle Strength}, B=LinearInterpolate(Alpha=ParticleRelativeTime()) {constA=10})
  n14 = Transform(None=MaterialFunctionCall(B=Multiply(A=@n12, B=@n13)) {materialFunction=&AppendMany}.RGB) {transformSourceType=TRANSFORMSOURCE_Local}
  Multiply(A=Add(A=Add(A=@n9) {constB=0}, B=Multiply(A=@n2.WigglePower, B=@n14)), B=@n4) {constA=1}
  Reroute(None=@n6.A)
 ]
}
