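# Niagara system for the DriedSilk pickaxe swing trails, layering three skeletal-mesh anim-trail ribbons: dark hatchings, sharp blade and white toon.
# Each emitter spans its own bone or socket pair, fading by a lifetime curve times the shared System.SwingGeneralFade.
# System.TODColor tints the blade and toon trails, while the hatchings stay black.
# The blade ribbon narrows with age; the toon trail has two renderers and writes a lower-alpha alternate color.
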NiagaraSystem /BRCosmetics/Weapons/FORT_Melee/Pickaxe_DriedSilk/FX/NS_Pickaxe_DriedSilk_Trails
dependencies {
 usedDataInterfaces [&NiagaraDataInterfaceCurve, &NiagaraDataInterfaceSkeletalMesh, &NiagaraDataInterfaceColorCurve]
 usedDynamicInputs [&ColorFromCurve, &FNTimeOfDayScalar]
 usedModules [
  &self:SystemSpawnScript.SystemScriptSource.SystemScriptGraph.NiagaraNodeAssignment_0.SetVariables_#1
  &SystemState
  &self:SystemSpawnScript.SystemScriptSource.SystemScriptGraph.NiagaraNodeAssignment_1.SetVariables_#2
  &self:SystemSpawnScript.SystemScriptSource.SystemScriptGraph.NiagaraNodeAssignment_2.SetVariables_#3
 ]
}
emitters {
 DarkHatchingsTrail {
  emitterUpdateScript [
   &EmitterState()
   &FN_SpawnAnimTrail(Spawn Interval=0)
  ]
  particleSpawnScript [
   &FN_InitializeAnimTrailBoneIndecies(First Bone or Socket Index=2, Second Bone or Socket Index=0)
   &FN_InitializeAnimTrail(AnimTrailWidth=0.6, Color=<rgba 0 0 0 1>, Lifetime=0.07, SkeletalMesh=$System.Skeletal Mesh) {module=FN_InitializeAnimTrail001}
   &CameraOffset(Camera Offset Amount=1) {enabled=false}
   &AddVelocity(Velocity=<0 1000 0>) {enabled=false}
  ]
  particleUpdateScript [
   &FN_UpdateAnimTrail(Locked Particle RibbonWidth=1, SkeletalMesh=$System.Skeletal Mesh) {module=FN_UpdateAnimTrail001}
   &ScaleColor(
    Scale Alpha=&Multiply_Float(
     A=$System.SwingGeneralFade
     B=&FloatFromCurve(FloatCurve=DataInterface() {propertyValues {Curve {keys [{arriveTangent=-0.0861, interpMode=RCIM_Cubic, leaveTangent=-0.0861, tangentMode=RCTM_User, time=0.7, value=1}, <at 1 0 RCIM_Cubic>]}}})
    )
    ScaleA=true
    ScaleRGB=false
   )
   &SolveForcesAndVelocity(Clamp Velocity=false, Limit Acceleration=false)
  ]
  rendererClasses [&NiagaraRibbonRendererProperties]
  renderers [{rendererClass=&NiagaraRibbonRendererProperties}]
 }
 SharpBladeTrail {
  emitterUpdateScript [
   &EmitterState()
   &FN_SpawnAnimTrail(AnimTrailWidthMode=::FromFirst, Spawn Interval=0)
  ]
  particleSpawnScript [
   &FN_InitializeAnimTrailBoneIndecies(First Bone or Socket Index=3, Second Bone or Socket Index=0)
   &FN_InitializeAnimTrail(AnimTrailWidth=0.55, Lifetime=0.3, SkeletalMesh=$System.Skeletal Mesh) {module=FN_InitializeAnimTrail001}
   &CameraOffset(Camera Offset Amount=2) {enabled=false}
   &AddVelocity(Velocity=<0 1000 0>) {enabled=false}
  ]
  particleUpdateScript [
   &FN_UpdateAnimTrail(Locked Particle RibbonWidth=1, SkeletalMesh=$System.Skeletal Mesh) {module=FN_UpdateAnimTrail001}
   &ScaleColor(
    Scale Alpha=&Multiply_Float(
     A=$System.SwingGeneralFade
     B=&FloatFromCurve(FloatCurve=DataInterface() {propertyValues {Curve {keys [
          {arriveTangent=-12.8154, interpMode=RCIM_Cubic, leaveTangent=-12.8154, tangentMode=RCTM_User, time=0.2, value=4}
          {arriveTangent=-3.1164, interpMode=RCIM_Cubic, leaveTangent=-3.1164, tangentMode=RCTM_User, time=0.45, value=1}
          <at 1 0 RCIM_Cubic>
         ]}}})
    )
    Scale RGB=&MakeVectorFromLinearColorRGB(Color=$System.TODColor)
    ScaleA=true
    ScaleRGB=true
   )
   &ScaleColor(
    Scale Alpha=&Multiply_Float(
     A=$System.SwingGeneralFade
     B=&FloatFromCurve(FloatCurve=DataInterface() {propertyValues {Curve {keys [
          {arriveTangent=-12.8154, interpMode=RCIM_Cubic, leaveTangent=-12.8154, tangentMode=RCTM_User, time=0.2, value=3}
          {arriveTangent=-3.1164, interpMode=RCIM_Cubic, leaveTangent=-3.1164, tangentMode=RCTM_User, time=0.45, value=1}
          <at 1 0 RCIM_Cubic>
         ]}}})
    )
    Scale RGB=&MakeVectorFromLinearColorRGB(Color=&FNTimeOfDayColor(
      Night Color=<rgba 0.1925 0.2371 0.35 1>
      Morning Color=<rgba 0.8 0.5527 0.52 1>
      Day Color=<rgba 1 1 1 1>
      Evening Color=<rgba 0.55 0.3575 0.3735 1>
     ))
    ScaleA=true
    ScaleRGB=true
   ) {enabled=false, module=ScaleColor001}
   &SolveForcesAndVelocity(Clamp Velocity=false, Limit Acceleration=false)
   &ScaleRibbonWidth(Ribbon Width Scale=&FloatFromCurve(FloatCurve=DataInterface() {propertyValues {Curve {keys [{interpMode=RCIM_Cubic, tangentMode=RCTM_User, time=0, value=1}, {interpMode=RCIM_Cubic, tangentMode=RCTM_User, time=0.3, value=0.1}]}}}))
  ]
  rendererClasses [&NiagaraRibbonRendererProperties]
  renderers [{rendererClass=&NiagaraRibbonRendererProperties}]
 }
 WhiteToonTrail {
  emitterUpdateScript [
   &EmitterState()
   &FN_SpawnAnimTrail(Spawn Interval=0)
  ]
  particleSpawnScript [
   &FN_InitializeAnimTrailBoneIndecies(First Bone or Socket Index=1, Second Bone or Socket Index=0)
   &FN_InitializeAnimTrail(Lifetime=0.065, SkeletalMesh=$System.Skeletal Mesh) {module=FN_InitializeAnimTrail001}
   &CameraOffset(Camera Offset Amount=-3)
   &AddVelocity(Velocity=<0 1000 0>) {enabled=false}
  ]
  particleUpdateScript [
   &FN_UpdateAnimTrail(Locked Particle RibbonWidth=1, SkeletalMesh=$System.Skeletal Mesh) {module=FN_UpdateAnimTrail001}
   &ScaleColor(
    Scale Alpha=&Multiply_Float(
     A=$System.SwingGeneralFade
     B=&FloatFromCurve(FloatCurve=DataInterface() {propertyValues {Curve {keys [{arriveTangent=-0.0783, interpMode=RCIM_Cubic, leaveTangent=-0.0783, tangentMode=RCTM_User, time=0.7, value=1}, <at 1 0 RCIM_Cubic>]}}})
    )
    Scale RGB=&MakeVectorFromLinearColorRGB(Color=$System.TODColor)
    ScaleA=true
    ScaleRGB=true
   )
   &ScaleColor(
    Scale Alpha=&Multiply_Float(
     A=$System.SwingGeneralFade
     B=&FloatFromCurve(FloatCurve=DataInterface() {propertyValues {Curve {keys [{arriveTangent=-0.0783, interpMode=RCIM_Cubic, leaveTangent=-0.0783, tangentMode=RCTM_User, time=0.7, value=1}, <at 1 0 RCIM_Cubic>]}}})
    )
    Scale RGB=&MakeVectorFromLinearColorRGB(Color=&FNTimeOfDayColor(Night Color=<rgba 0.1 0.14 0.25 1>, Morning Color=<rgba 0.4 0.235 0.22 1>, Day Color=<rgba 1 1 1 1>, Evening Color=<rgba 0.4 0.2267 0.2 1>))
    ScaleA=true
    ScaleRGB=true
   ) {enabled=false, module=ScaleColor001}
   /SetVariables(Particles.ColorTrailAlt:LinearColor=&MultiplyLinearColor(Linear Color=$Particles.Color, Scale Factor=<1 1 1 0.23>)) {isSetParameters=true}
   &SolveForcesAndVelocity(Clamp Velocity=false, Limit Acceleration=false)
  ]
  rendererClasses [&NiagaraRibbonRendererProperties]
  renderers [
   {rendererClass=&NiagaraRibbonRendererProperties}
   {rendererClass=&NiagaraRibbonRendererProperties, rendererIndex=1}
  ]
 }
}
summary {
 systemName=NS_Pickaxe_DriedSilk_Trails
 emitters [
  {emitterName=WhiteToonTrail, rendererClasses [&NiagaraRibbonRendererProperties]}
  {emitterName=DarkHatchingsTrail, rendererClasses [&NiagaraRibbonRendererProperties]}
  {emitterName=SharpBladeTrail, rendererClasses [&NiagaraRibbonRendererProperties]}
 ]
 userVariables [
  User.bisSwinging {type=NiagaraBool}
  User.NewVector3f {type=Vector3f, description="", defaultValue {struct=Vector3f, value=<0 0 0>}}
  User.RibbonIDOffset {type=NiagaraInt32}
 ]
}
