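# Physics asset for the DriedSilk two-handed pickaxe skeletal mesh, a melee weapon cosmetic.
# It holds a single body on the Melee_TwoHanded_Handle bone, shaped as one capsule running along the handle.
# No constraints are defined, so the weapon simulates as one rigid piece.
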PhysicsAsset /BRCosmetics/Weapons/FORT_Melee/Pickaxe_DriedSilk/Meshes/DriedSilk_Axe_PhysicsAsset
bodies {Melee_TwoHanded_Handle [{shapeType=Capsule, length=51.2389, radius=5.7393, center=<-0.262 -0.1378 26.8748>, rotation=<rot 1.4785 -90 -0.0299>}]}
constraints []
